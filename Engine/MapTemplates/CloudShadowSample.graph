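# CloudShadowSample is a material function that computes one cloud self-shadowing sample from a density texture and light direction.
# It offsets CloudCoordinates along LightDirection's XY by SampleDistance, then raises the sampled density to DensityPower.
# That density is scaled by inverted, clamped light Z and SampleOpacity, then inverted as the output, so shadowing fades as light Z rises.
# A trailing distance-weighting multiply is left unconnected.
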MaterialFunction /Engine/MapTemplates/CloudShadowSample
graph {nodes [
  Constant() {r=0.8}
  n1 = Constant3Vector() {desc=Sky Vector, constant=<rgba 0 0 1 0>}
  n2 = FunctionInput(Preview=Transform(None=Normalize(VectorInput=@n1)) {transformSourceType=TRANSFORMSOURCE_World, transformType=TRANSFORM_Tangent}) {inputName=LightDirection}
  n3 = Clamp(None=Add(A=ComponentMask(None=@n2) {b=true}, B=Constant()))
  n4 = ComponentMask(None=@n2) {g=true, r=true}
  n5 = FunctionInput(Preview=Constant() {desc=Sample_Distance, r=0.01}) {inputName=SampleDistance, inputType=FunctionInput_Scalar}
  n6 = TextureCoordinate()
  n7 = TextureSample(
   UVs=Add(
    A=Multiply(A=@n4, B=@n5)
    B=FunctionInput(Preview=@n6) {bUsePreviewValueAsDefault=true, inputName=CloudCoordinates, inputType=FunctionInput_Vector2}
   )
   Tex=FunctionInput(Preview=TextureObject() {texture=null}) {inputName=CloudDensity, inputType=FunctionInput_Texture2D}
  ) {samplerType=SAMPLERTYPE_Grayscale}
  n8 = FunctionInput(Preview=Constant() {r=1}) {bUsePreviewValueAsDefault=true, inputName=DensityPower, inputType=FunctionInput_Scalar}
  n9 = Multiply(A=Multiply(A=OneMinus(None=@n3), B=Power(Base=@n7.R, Exp=@n8)), B=FunctionInput() {inputName=SampleOpacity, inputType=FunctionInput_Scalar})
  FunctionOutput(None=OneMinus(None=@n9))
  Multiply(A=@n9, B=Multiply(A=@n5, B=Constant() {r=10}))
 ]}
