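# Material function that computes a cloud shadow term from four CloudShadowSample taps on a density texture.
# Each tap steps the sample distance out from SampleDistanceStart by SampleDistanceOffset.
# The light direction is pushed to alternating sides along X (+1, -1, +2, -2), scaled by that tap's distance, spreading the samples around the light.
# All taps share opacity, coordinates and density power, and their results are multiplied into one output.
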MaterialFunction /Engine/MapTemplates/CloudShadowSample4Iterations_Hem
graph {nodes [
  n1 = FunctionInput() {inputName=LightDirection}
  n2 = FunctionInput() {inputName=SampleDistanceStart, inputType=FunctionInput_Scalar}
  n3 = AppendVector(A=Constant2Vector() {r=1}, B=Constant())
  n4 = FunctionInput() {inputName=SampleOpacity, inputType=FunctionInput_Scalar}
  n5 = FunctionInput(Preview=TextureCoordinate()) {bUsePreviewValueAsDefault=true, inputName=CloudCoordinates, inputType=FunctionInput_Vector2}
  n6 = FunctionInput(Preview=Constant() {r=1}) {bUsePreviewValueAsDefault=true, inputName=DensityPower, inputType=FunctionInput_Scalar}
  n7 = FunctionInput(Preview=TextureObject() {samplerType=SAMPLERTYPE_Grayscale, texture=&/TestCard}) {inputName=CloudDensity, inputType=FunctionInput_Texture2D}
  n8 = MaterialFunctionCall!(
   LightDirection=Add(A=@n1, B=Multiply(A=@n2, B=@n3))
   SampleDistance=@n2
   SampleOpacity=@n4
   CloudCoordinates=@n5
   DensityPower=@n6
   CloudDensity=@n7
  ) {materialFunction=&/CloudShadowSample}
  n9 = FunctionInput() {inputName=SampleDistanceOffset, inputType=FunctionInput_Scalar}
  n10 = Add(A=@n2, B=@n9)
  n11 = AppendVector(A=Constant2Vector() {r=-1}, B=Constant())
  n12 = MaterialFunctionCall(
   LightDirection=Add(A=@n1, B=Multiply(A=@n10, B=@n11))
   SampleDistance=@n10
   SampleOpacity=@n4
   CloudCoordinates=@n5
   DensityPower=@n6
   CloudDensity=@n7
  ) {materialFunction=&CloudShadowSample}
  n13 = Add(A=@n10, B=@n9)
  n14 = AppendVector(A=Constant2Vector() {r=2}, B=Constant())
  n15 = MaterialFunctionCall(
   LightDirection=Add(A=@n1, B=Multiply(A=@n13, B=@n14))
   SampleDistance=@n13
   SampleOpacity=@n4
   CloudCoordinates=@n5
   DensityPower=@n6
   CloudDensity=@n7
  ) {materialFunction=&CloudShadowSample}
  n16 = Add(A=@n13, B=@n9)
  n17 = AppendVector(A=Constant2Vector() {r=-2}, B=Constant())
  n18 = MaterialFunctionCall(
   LightDirection=Add(A=@n1, B=Multiply(A=@n16, B=@n17))
   SampleDistance=@n16
   SampleOpacity=@n4
   CloudCoordinates=@n5
   DensityPower=@n6
   CloudDensity=@n7
  ) {materialFunction=&CloudShadowSample}
  FunctionOutput(None=Multiply(A=Multiply(A=@n8.Result, B=@n12.Result), B=Multiply(A=@n15.Result, B=@n18.Result)))
 ]}
