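# Material function that computes a cloud self-shadow term by taking four CloudShadowSample samples along a light direction.
# Each call shares the direction, opacity, cloud coordinates, density power and cloud texture input.
# Sample distance begins at SampleDistanceStart and grows by SampleDistanceOffset per step.
# The four results multiply together into one output, forming an unrolled fixed-step march.
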MaterialFunction /Engine/MapTemplates/CloudShadowSample4Iterations
graph {nodes [
  n1 = FunctionInput() {inputName=LightDirection}
  n2 = FunctionInput() {inputName=SampleDistanceStart, inputType=FunctionInput_Scalar}
  n3 = FunctionInput() {inputName=SampleOpacity, inputType=FunctionInput_Scalar}
  n4 = FunctionInput(Preview=TextureCoordinate()) {bUsePreviewValueAsDefault=true, inputName=CloudCoordinates, inputType=FunctionInput_Vector2}
  n5 = FunctionInput(Preview=Constant() {r=1}) {bUsePreviewValueAsDefault=true, inputName=DensityPower, inputType=FunctionInput_Scalar}
  n6 = FunctionInput(Preview=TextureObject() {texture=null}) {inputName=Texture2D'ThroneSky.Textures.T_UDK_Sky_CloudMask01', inputType=FunctionInput_Texture2D}
  n7 = MaterialFunctionCall!(LightDirection=@n1, SampleDistance=@n2, SampleOpacity=@n3, CloudCoordinates=@n4, DensityPower=@n5, CloudDensity=@n6) {materialFunction=&/CloudShadowSample}
  n8 = FunctionInput() {inputName=SampleDistanceOffset, inputType=FunctionInput_Scalar}
  n9 = Add(A=@n2, B=@n8)
  n10 = MaterialFunctionCall(LightDirection=@n1, SampleDistance=@n9, SampleOpacity=@n3, CloudCoordinates=@n4, DensityPower=@n5, CloudDensity=@n6) {materialFunction=&CloudShadowSample}
  n11 = Add(A=@n9, B=@n8)
  n12 = MaterialFunctionCall(LightDirection=@n1, SampleDistance=@n11, SampleOpacity=@n3, CloudCoordinates=@n4, DensityPower=@n5, CloudDensity=@n6) {materialFunction=&CloudShadowSample}
  n13 = MaterialFunctionCall(LightDirection=@n1, SampleDistance=Add(A=@n11, B=@n8), SampleOpacity=@n3, CloudCoordinates=@n4, DensityPower=@n5, CloudDensity=@n6) {materialFunction=&CloudShadowSample}
  FunctionOutput(None=Multiply(A=Multiply(A=@n7.Result, B=@n10.Result), B=Multiply(A=@n12.Result, B=@n13.Result)))
 ]}
