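# Material function that computes cloud self-shadowing from a CloudDensity texture by sampling along the light direction.
# The normalized LightVector is transformed from world to tangent space and passed with CloudCoordinates and DensityPower to four CloudShadowSample4Iterations calls.
# The calls start at evenly staggered distances and share one step offset, both scaled by ShadowLength, forming an unrolled march.
# Their results are multiplied together and raised to ShadowFalloff for the output.
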MaterialFunction /Engine/MapTemplates/CloudDensityShadowing
graph {nodes [
  n1 = FunctionInput(Preview=Constant() {r=1}) {bUsePreviewValueAsDefault=true, inputName=ShadowLength, inputType=FunctionInput_Scalar}
  n2 = Multiply(A=Constant() {r=0.01}, B=@n1)
  n3 = FunctionInput(Preview=TextureObject() {texture=null}) {inputName=CloudDensity, inputType=FunctionInput_Texture2D}
  n4 = Constant() {r=0.3}
  n5 = FunctionInput(Preview=TextureCoordinate()) {bUsePreviewValueAsDefault=true, inputName=CloudCoordinates, inputType=FunctionInput_Vector2}
  n6 = FunctionInput(Preview=Constant() {r=1}) {bUsePreviewValueAsDefault=true, inputName=DensityPower, inputType=FunctionInput_Scalar}
  n7 = FunctionInput(Preview=Constant3Vector() {desc=Sky Vector, constant=<rgba 0 0 1 0>}) {bUsePreviewValueAsDefault=true, inputName=LightVector}
  n8 = Add(A=Transform(None=Normalize(VectorInput=@n7)) {transformSourceType=TRANSFORMSOURCE_World, transformType=TRANSFORM_Tangent}, B=Constant())
  n9 = Multiply(A=Constant() {r=0.05}, B=@n1)
  n10 = Constant() {r=0.09}
  n11 = Multiply(
   A=Multiply(
    A=MaterialFunctionCall!(
     SampleDistanceStart=@n2
     Texture2D'ThroneSky.Textures.T_UDK_Sky_CloudMask01'=@n3
     SampleOpacity=@n4
     SampleDistanceOffset=@n2
     CloudCoordinates=@n5
     DensityPower=@n6
     LightDirection=@n8
    ) {materialFunction=&/CloudShadowSample4Iterations}.Result
    B=MaterialFunctionCall(
     SampleDistanceStart=@n9
     Texture2D'ThroneSky.Textures.T_UDK_Sky_CloudMask01'=@n3
     SampleOpacity=@n4
     SampleDistanceOffset=@n2
     CloudCoordinates=@n5
     DensityPower=@n6
     LightDirection=@n8
    ) {materialFunction=&CloudShadowSample4Iterations}.Result
   )
   B=MaterialFunctionCall(
    SampleDistanceStart=Multiply(A=@n10, B=@n1)
    Texture2D'ThroneSky.Textures.T_UDK_Sky_CloudMask01'=@n3
    SampleOpacity=@n4
    SampleDistanceOffset=@n2
    CloudCoordinates=@n5
    DensityPower=@n6
    LightDirection=@n8
   ) {materialFunction=&CloudShadowSample4Iterations}.Result
  )
  n12 = MaterialFunctionCall(
   SampleDistanceStart=Multiply(A=Constant() {r=0.13}, B=@n1)
   Texture2D'ThroneSky.Textures.T_UDK_Sky_CloudMask01'=@n3
   SampleOpacity=@n4
   SampleDistanceOffset=@n2
   CloudCoordinates=@n5
   DensityPower=@n6
   LightDirection=@n8
  ) {materialFunction=&CloudShadowSample4Iterations}
  n13 = Constant() {r=1}
  FunctionOutput(None=Power(
    Base=Multiply(A=@n11, B=@n12.Result)
    Exp=FunctionInput(Preview=@n13) {bUsePreviewValueAsDefault=true, inputName=ShadowFalloff, inputType=FunctionInput_Scalar}
   ))
 ]}
