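# Material function that computes a scalar rim lighting mask for clouds, strongest where thin cloud is backlit by the light.
# The negated dot product of camera and light vectors is clamped and raised to a fixed falloff exponent.
# Cloud density, shaped by an optional FalloffPower input, is inverted in 2.2 gamma space and converted back.
# The output multiplies the view-to-light term by this inverted density.
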MaterialFunction /Engine/MapTemplates/CloudRimLighting
graph {nodes [
  n1 = CameraVectorWS()
  n2 = FunctionInput(Preview=Normalize(VectorInput=Constant3Vector() {desc=LIGHT_VECTOR, constant=<rgba 1 0 1 0>})) {inputName=NormalizedLightVector}
  n3 = Clamp(None=Multiply(A=DotProduct(A=@n1, B=@n2), B=Constant() {r=-1}))
  n4 = Constant() {desc=RimLightingFalloff, r=4}
  n5 = FunctionInput(Preview=TextureSample() {desc=OPACITY_DENSITY, samplerType=SAMPLERTYPE_Grayscale, texture=&/T_UDK_Sky_CloudMask03}.R) {inputName=Density, inputType=FunctionInput_Scalar}
  n6 = FunctionInput(Preview=Constant() {r=1}) {bUsePreviewValueAsDefault=true, inputName=FalloffPower, inputType=FunctionInput_Scalar}
  n7 = Constant() {r=1}
  n8 = Constant() {r=2.2}
  n9 = OneMinus(None=Power(Base=Power(Base=@n5, Exp=@n6), Exp=Divide(A=@n7, B=@n8)))
  FunctionOutput(None=Multiply(A=Power(Base=@n3, Exp=@n4), B=Power(Base=@n9, Exp=@n8)))
 ]}
