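# Unlit master sky material that renders a horizon-to-zenith gradient, sun disc and animated, lit clouds.
# A panning cloud mask, its power varied by scrolling noise, feeds sun-shadow, hemisphere-shadow and rim-lighting functions, and blends clouds over sky and sun.
# DarkCloudLayer drives blends to dark colors and grading, but both its branches are zero constants.
# Static debug switches isolate sky, clouds or either shading term.
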Material /Engine/MapTemplates/MASTER_Sky_Material
graph {
 properties {bUsedWithStaticLighting=true, pixelDepthOffsetMode=PDOM_Legacy, shadingModel=MSM_Unlit, usedShadingModels=Unlit}
 nodes [
  n1 = Normalize(VectorInput=VectorParameter() {group=Sun, parameterName=LightDir, defaultValue=<rgba 2 0 1 1>}.RGB)
  n2 = TextureObjectParameter() {group=CloudTextures, parameterName=CloudDensity, samplerType=SAMPLERTYPE_Grayscale, texture=&/T_UDK_Sky_CloudMask03}
  n3 = Panner(Time=Multiply(A=Time(), B=ScalarParameter() {defaultValue=0.01, group=Animation, parameterName=PannerSpeed})) {speedX=1}
  n4 = StaticSwitchParameter(True=Constant(), False=Constant()) {group=Global, parameterName=DarkCloudLayer}
  n5 = Panner(
   Coordinate=Multiply(A=TextureCoordinate() {coordinateIndex=2}, B=ScalarParameter() {defaultValue=1, group=Animation, parameterName=NoiseTiling})
   Time=Multiply(A=Time(), B=ScalarParameter() {defaultValue=0.02, group=Animation, parameterName=NoisePannerSpeed})
  ) {speedX=1}
  n6 = Multiply(
   A=LinearInterpolate(A=Constant() {r=1}, B=ScalarParameter() {defaultValue=1, parameterName=DarkCloudPowerScale}, Alpha=@n4)
   B=LinearInterpolate(
    A=ScalarParameter() {defaultValue=1, group=Animation, parameterName=NoisePower1}
    B=ScalarParameter() {defaultValue=4, group=Animation, parameterName=NoisePower2}
    Alpha=TextureSampleParameter2D(UVs=@n5) {group=Animation, parameterName=NoiseTexture, texture=&/TilingNoise05}.R
   ) {constA=0.1, constB=4}
  )
  n7 = MaterialFunctionCall!(
   LightVector=@n1
   ShadowLength=ScalarParameter() {defaultValue=0.2, group=CloudShadows, parameterName=CloudShadowLength}
   CloudDensity=@n2
   CloudCoordinates=@n3
   DensityPower=@n6
   ShadowFalloff=ScalarParameter() {defaultValue=2, group=CloudShadows, parameterName=CloudShadowFalloff}
  ) {materialFunction=&/CloudDensityShadowing}
  n8 = Constant3Vector() {constant=<rgba 0 0 1 0>}
  n9 = ScalarParameter() {defaultValue=1.5, group=CloudShadows, parameterName=CloudHemisphereShadowFalloff}
  n10 = MaterialFunctionCall!(
   ShadowLength=ScalarParameter() {defaultValue=0.1, group=CloudShadows, parameterName=CloudHemisphereShadowLength1}
   LightVector=@n8
   ShadowFalloff=@n9
   CloudCoordinates=@n3
   DensityPower=@n6
   CloudDensity=@n2
  ) {materialFunction=&/CloudHemisphereShadowing}
  n11 = MaterialFunctionCall(
   ShadowLength=ScalarParameter() {defaultValue=0.2, group=CloudShadows, parameterName=CloudHemisphereShadowLength2}
   LightVector=@n8
   ShadowFalloff=@n9
   CloudCoordinates=@n3
   DensityPower=@n6
   CloudDensity=@n2
  ) {materialFunction=&CloudHemisphereShadowing}
  n12 = Multiply(A=Add(A=OneMinus(None=@n10.Result), B=OneMinus(None=@n11.Result)), B=Constant() {r=0.5})
  n13 = TextureSampleParameter2D(UVs=@n3) {group=CloudTextures, parameterName=CloudOpacity, samplerType=SAMPLERTYPE_Grayscale, texture=&T_UDK_Sky_CloudMask03}
  n14 = Power(Base=@n13.R, Exp=@n6)
  n15 = MaterialFunctionCall!(Density=@n14, FalloffPower=Constant() {r=1}, NormalizedLightVector=@n1) {materialFunction=&/CloudRimLighting}
  n16 = ScalarParameter() {defaultValue=40, group=CloudColors, parameterName=RimLightingIntensity}
  n17 = LinearInterpolate(
   A=VectorParameter() {group=CloudColors, parameterName=RimLightColor, defaultValue=<rgba 1 0.8801 0.7964 0>}.RGB
   B=VectorParameter() {group=DarkCloudColors, parameterName=DarkRimLightColor, defaultValue=<rgba 1 0.8801 0.7964 0>}.RGB
   Alpha=@n4
  )
  n18 = LinearInterpolate(
   A=VectorParameter() {group=CloudColors, parameterName=CloudLightColor, defaultValue=<rgba 1 1 1 0>}.RGB
   B=VectorParameter() {group=DarkCloudColors, parameterName=DarkCloudLightColor, defaultValue=<rgba 1 1 1 0>}.RGB
   Alpha=@n4
  )
  n19 = OneMinus(None=@n12)
  n20 = Multiply(A=Multiply(A=Multiply(A=@n15.Result, B=@n16), B=Multiply(A=@n17, B=@n18)), B=@n19)
  n21 = Multiply(A=OneMinus(None=@n19), B=ScalarParameter() {defaultValue=0.9, group=CloudShadows, parameterName=CloudHemisphereShadowIntensity})
  n22 = VectorParameter() {group=DarkCloudColors, parameterName=DarkCloudSkylightGroundColor, defaultValue=<rgba 0.66 0.5965 0.5544 0>}
  n23 = LinearInterpolate(
   A=VectorParameter() {group=CloudColors, parameterName=CloudSkylightGroundColor, defaultValue=<rgba 0.66 0.5965 0.5544 0>}.RGB
   B=@n22.RGB
   Alpha=@n4
  )
  n24 = LinearInterpolate(
   A=VectorParameter() {group=CloudColors, parameterName=CloudSkyllightSkyColor, defaultValue=<rgba 0.7465 0.907 1 0>}.RGB
   B=VectorParameter() {group=DarkCloudColors, parameterName=DarkCloudSkyllightSkyColor, defaultValue=<rgba 0.7465 0.907 1 0>}.RGB
   Alpha=@n4
  )
  n25 = Multiply(A=@n19, B=@n19)
  n26 = Multiply(
   A=Multiply(A=OneMinus(None=@n21), B=LinearInterpolate(A=@n23, B=@n24, Alpha=@n25))
   B=ScalarParameter() {defaultValue=1, group=CloudColors, parameterName=CloudSkyLightBrightness}
  )
  n27 = LinearInterpolate(
   A=LinearInterpolate(
    A=VectorParameter() {group=CloudColors, parameterName=CloudScatterColor, defaultValue=<rgba 0.716 0.8764 1 0>}.RGB
    B=VectorParameter() {group=DarkCloudColors, parameterName=DarkCloudScatterColor, defaultValue=<rgba 0.716 0.8764 1 0>}.RGB
    Alpha=@n4
   )
   B=@n18
   Alpha=Multiply(A=@n7.Result, B=@n7.Result)
  )
  n28 = Multiply(
   A=Add(A=Add(A=@n20, B=@n26), B=Multiply(A=@n27, B=@n7.Result))
   B=ScalarParameter() {defaultValue=1.5, group=CloudColors, parameterName=CloudBrightness}
  )
  n29 = LinearInterpolate(
   A=VectorParameter() {group=Sky, parameterName=SkyZenithColor, defaultValue=<rgba 0.0546 0.0626 0.1 1>}.RGB
   B=VectorParameter() {group=Sky, parameterName=DarkSkyZenithColor, defaultValue=<rgba 0.0546 0.0626 0.1 1>}.RGB
   Alpha=@n4
  )
  n30 = DotProduct(A=CameraVectorWS(), B=Constant3Vector() {constant=<rgba 0 0 -1 0>})
  n31 = Power(Base=OneMinus(None=Clamp(None=@n30)), Exp=ScalarParameter() {defaultValue=3, group=Sky, parameterName=SkyHorizonFalloff})
  n32 = OneMinus(None=@n31)
  n33 = CameraVectorWS()
  n34 = Clamp(None=Multiply(A=DotProduct(A=@n33, B=@n1), B=Constant() {r=-1}))
  n35 = Constant() {r=1}
  n36 = ScalarParameter() {defaultValue=0.001, group=Sun, parameterName=SunRadius}
  n37 = VectorParameter() {group=Sun, parameterName=SunColor, defaultValue=<rgba 10 6 2 1>}
  n38 = Add(
   A=Multiply(A=Multiply(A=@n29, B=@n32), B=ScalarParameter() {defaultValue=1, group=Sky, parameterName=SkyBrightness})
   B=Multiply(A=SphereMask(A=@n34, B=@n35, Radius=@n36) {attenuationRadius=0.001, hardnessPercent=0.95}, B=Multiply(A=@n37.RGB, B=@n37.RGB))
  )
  n39 = Multiply(
   A=LinearInterpolate(
    A=VectorParameter() {group=Sky, parameterName=SkyHorizonColor, defaultValue=<rgba 0.3184 0.3841 0.5 1>}.RGB
    B=VectorParameter() {group=Sky, parameterName=DarkSkyHorizonColor, defaultValue=<rgba 0.3184 0.3841 0.5 1>}.RGB
    Alpha=@n4
   )
   B=@n31
  )
  n40 = LinearInterpolate(A=@n38, B=@n28, Alpha=@n14)
  n41 = Multiply(
   A=Multiply(A=Add(A=@n39, B=@n40), B=ScalarParameter() {defaultValue=1, group=Global, parameterName=GlobalBrightness})
   B=VectorParameter() {group=Global, parameterName=GlobalColor, defaultValue=<rgba 1 1 1 1>}.RGB
  )
  n42 = StaticSwitchParameter(
   True=@n12
   False=StaticSwitchParameter(True=@n28, False=StaticSwitchParameter(True=@n38, False=@n41) {group=Debug, parameterName=ShowSkyOnly}) {group=Debug, parameterName=ShowCloudsOnly}
  ) {group=Debug, parameterName=ShowHemisphereShadingOnly}
  n43 = ScalarParameter() {defaultValue=0.5, parameterName=DarkSkySaturation}
  n44 = Multiply(
   A=Desaturation(
    None=StaticSwitchParameter(True=@n7.Result, False=@n42) {group=Debug, parameterName=ShowSunShadingOnly}
    Fraction=LinearInterpolate(B=@n43, Alpha=@n4)
   )
   B=LinearInterpolate(B=ScalarParameter() {defaultValue=1, parameterName=DarkSkyBrightness}, Alpha=@n4) {constA=1}
  )
  n45 = Clamp(None=@n44)
  n46 = ScalarParameter() {defaultValue=1, parameterName=DarkSkyPower}
  Multiply(A=LinearInterpolate(B=Power(Base=@n45, Exp=@n46), Alpha=@n4) {constA=1}, B=@n44)
 ]
}
