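# BasicAsset01 is a simple Default Lit engine material in the map templates folder, flagged for static lighting, skeletal meshes and mesh deformers.
# Its graph holds only two constants, a red-orange color vector and a scalar, with no textures, parameters or math.
# The graph does not show which material inputs these constants drive.
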Material /Engine/MapTemplates/Materials/BasicAsset01
graph {
 properties {
  bUsedWithMeshDeformer=true
  bUsedWithSkeletalMesh=true
  bUsedWithStaticLighting=true
  pixelDepthOffsetMode=PDOM_Legacy
  usedShadingModels=Default Lit
 }
 nodes [
  Constant3Vector() {constant=<rgba 1 0.05 0 0>}
  Constant() {r=0.8}
 ]
}
