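# BasicAsset02 is a simple Default Lit material in the engine's map template content, flagged for use with static lighting.
# Its graph holds only three constant nodes: a near-white, faintly cool grey color and two scalar values.
# It uses no textures, parameters or math, and the file does not show which material inputs the constants feed.
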Material /Engine/MapTemplates/Materials/BasicAsset02
graph {
 properties {bUsedWithStaticLighting=true, pixelDepthOffsetMode=PDOM_Legacy, usedShadingModels=Default Lit}
 nodes [
  Constant3Vector() {constant=<rgba 0.913 0.921 0.925 0>}
  Constant() {r=1}
  Constant() {r=0.4}
 ]
}
