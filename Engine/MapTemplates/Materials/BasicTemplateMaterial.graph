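# Basic Default Lit material in the engine's MapTemplates content, flagged for use with static lighting.
# The graph holds only two scalar Constant nodes, with no textures or parameters and no visible links to material inputs.
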Material /Engine/MapTemplates/Materials/BasicTemplateMaterial
graph {
 properties {bUsedWithStaticLighting=true, pixelDepthOffsetMode=PDOM_Legacy, usedShadingModels=Default Lit}
 nodes [
  Constant() {r=0.4}
  Constant() {r=0.2}
 ]
}
