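# Example material layer function that outputs a basic surface as Material Attributes, using the engine's default diffuse and normal textures.
# A Tiling input scales the UVs shared by both texture samples, and a Tint input multiplies the base color.
# The diffuse texture's alpha drives roughness through a lerp between min and max constants, while metalness is a constant.
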MaterialFunction /Engine/MapTemplates/Materials/ML_ExampleMaterialLayer
graph {nodes [
  n1 = FunctionInput() {bUsePreviewValueAsDefault=true, description=Tint the base Color, inputName=Tint, previewValue=<1 1 1 1>}
  n2 = FunctionInput() {bUsePreviewValueAsDefault=true, inputName=Tiling, inputType=FunctionInput_Scalar, previewValue=<1 0 0 1>}
  n3 = Multiply(A=TextureCoordinate(), B=@n2)
  n4 = TextureSample(UVs=@n3) {desc=BaseColor in RGB Roughness in Alpha, texture=/Engine/EngineMaterials/DefaultDiffuse}
  n5 = Constant() {desc=RoughnessMIN}
  n6 = Constant() {desc=RoughnessMAX, r=1}
  FunctionOutput(None=MakeMaterialAttributes!(
    BaseColor=Multiply(A=@n1, B=@n4.RGB)
    Metallic=Constant() {desc=Metalness}
    Specular
    Roughness=LinearInterpolate(A=@n5, B=@n6, Alpha=@n4.A)
    Anisotropy
    EmissiveColor
    Opacity
    OpacityMask
    Normal=TextureSample(UVs=@n3) {desc=NormalMap, samplerType=SAMPLERTYPE_Normal, texture=/Engine/EngineMaterials/DefaultNormal}.RGB
    Tangent
    WorldPositionOffset
    SubsurfaceColor
    ClearCoat
    ClearCoatRoughness
    AmbientOcclusion
    Refraction
    CustomizedUVs_0
    CustomizedUVs_1
    CustomizedUVs_2
    CustomizedUVs_3
    CustomizedUVs_4
    CustomizedUVs_5
    CustomizedUVs_6
    CustomizedUVs_7
    PixelDepthOffset
    ShadingModel
    Displacement
   ))
 ]}
